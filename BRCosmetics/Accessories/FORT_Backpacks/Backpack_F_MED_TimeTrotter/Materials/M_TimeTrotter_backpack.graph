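# Clothing-ready skeletal mesh material for the TimeTrotter backpack, adding an animated clock and scrolling emissive gradients between the 2023 character parent PreFX and PostFX functions.
# Second-UV regions separate main wheel, side wheel and wing, with the main wheel using polar coordinates for gradient panning.
# The clock function, sphere-masked in pre-skinned local space, replaces dark emissive areas within tech-art masks.
# That sphere also drives temporal responsiveness, and a car clear-coat function finishes shading.
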Material /BRCosmetics/Accessories/FORT_Backpacks/Backpack_F_MED_TimeTrotter/Materials/M_TimeTrotter_backpack
graph {
 properties {
  bUseMaterialAttributes=true
  bUsedWithClothing=true
  bUsedWithMeshDeformer=true
  bUsedWithMorphTargets=true
  bUsedWithSkeletalMesh=true
  pixelDepthOffsetMode=PDOM_Legacy
  previewMesh=/BRCosmetics/Accessories/FORT_Backpacks/Backpack_F_MED_TimeTrotter/Meshes/F_MED_TimeTrotter_Pack
  shadingModel=MSM_FromMaterialExpression
  usedShadingModels=Default Lit | Subsurface | Clear Coat | Subsurface Profile | Cloth
 }
 nodes [
  n1 = MaterialFunctionCall() {materialFunction=&MF_CharacterParent_2023_PreFX}
  n2 = GetMaterialAttributes(None=@n1.Result) {attributeGetTypes [#1]}
  n3 = LinearInterpolate(A=NamedRerouteUsage(), B=@n2.Emissive Color, Alpha=NamedRerouteUsage())
  n4 = SphereMask(A=@n1.Pre skinned Local Position, B=VectorParameter() {group=TimeTrotter_Clock, parameterName=Position, defaultValue=<rgba 0 0 -2.5 1>}.RGB) {attenuationRadius=5}
  n5 = Multiply(
   A=MaterialFunctionCall() {materialFunction=&/MF_TimeTrotter_Clock}.Clock
   B=@n4
  )
  n6 = ComponentMask(None=Desaturation(None=@n2.Emissive Color, Fraction=Constant() {r=1})) {r=true}
  n7 = TextureSampleParameter2D!(UVs=TextureCoordinate(), Level, Apply View MipBias) {
   group=TimeTrotter
   mipValueMode=TMVM_MipLevel
   parameterName=TechArtMap
   samplerSource=SSM_Wrap_WorldGroupSettings
   samplerType=SAMPLERTYPE_Masks
   texture=&T_BlackMask
  }
  n8 = Multiply(
   A=OneMinus(None=SmoothStep(Value=@n6) {constMax=0.105, constMin=0.09})
   B=Subtract(A=ComponentMask(None=@n1.Tech Art Mask) {r=true}, B=@n7.G)
  )
  n9 = MaterialFunctionCall!(
   Mask=@n1.Tech Art Mask
   Fresnel=@n1.Tech Art Mask
   In=SetMaterialAttributes(MaterialAttributes=@n2.MaterialAttributes, Emissive Color=LinearInterpolate(A=@n3, B=@n5, Alpha=@n8)) {attributeSetTypes [#1]}
  ) {materialFunction=/Game/Global/MaterialFunctions/TechArt/MF_Advanced_CarClearCoat_v2}
  ClearCoatNormalCustomOutput!(None=@n9.ClearCoatNormal)
  MaterialFunctionCall(Material=@n9.Result, M=@n1.M, Pre Skinned Local Position=@n1.M, Distance Blend=@n1.M, Fresnel=@n1.M) {materialFunction=&MF_CharacterParent_2023_PostFX}
  NamedRerouteDeclaration(None=@n1.Fresnel) {name=Fresnel, nodeColor=<rgba 1 0 0.7734 1>}
  n10 = NamedRerouteUsage()
  n11 = Power(
   Base=LinearInterpolate(A=OneMinus(None=@n10), B=@n10, Alpha=ScalarParameter() {group=TimeTrotter, parameterName=Emissive_Fresnel, sortPriority=39})
   Exp=ScalarParameter() {defaultValue=2, group=TimeTrotter, parameterName=Emissive_FresnelPower, sortPriority=39}
  )
  NamedRerouteDeclaration(None=Multiply(A=NamedRerouteUsage(), B=Saturate(None=@n11))) {name=UV2_Emissive, nodeColor=<rgba 0.6172 0 1 1>}
  n12 = Multiply(A=@n7.B, B=ScalarParameter() {defaultValue=0.5, group=TimeTrotter, parameterName=loops_WingSub})
  n13 = ScalarParameter() {defaultValue=0.1, group=TimeTrotter, parameterName=EmissiveGradient.WheelSpeed}
  n14 = ScalarParameter() {defaultValue=0.5, group=TimeTrotter, parameterName=EmissiveGradient.PanningSpeed}
  n15 = TextureCoordinate() {coordinateIndex=1}
  n16 = ComponentMask(None=@n15) {r=true}
  n17 = Step(Y=@n16) {constX=0.438}
  n18 = TextureCoordinate() {coordinateIndex=1}
  n19 = ComponentMask(None=Constant4Vector() {constant=<rgba 0.7125 0.285 0 0>}.RGBA) {g=true, r=true}
  n20 = Subtract(A=@n18, B=@n19) {constB=0.5}
  n21 = ComponentMask(None=@n20) {r=true}
  n22 = ComponentMask(None=@n20) {g=true}
  n23 = Add(A=Divide(A=Arctangent2Fast(Y=@n21, X=@n22)) {constB=6.2831}) {constB=0.5}
  n24 = Constant3Vector() {constant=<rgba 0 0 1 0>}
  n25 = Constant3Vector() {constant=<rgba 0 1 0 0>}
  n26 = Constant3Vector() {constant=<rgba 1 0 0 0>}
  n27 = LinearInterpolate(A=@n24, B=LinearInterpolate(A=@n25, B=@n26, Alpha=Step(Y=@n16) {constX=0.3125}), Alpha=@n17)
  n28 = Multiply(A=DotProduct(A=@n24, B=@n27), B=ScalarParameter() {defaultValue=1, group=TimeTrotter, parameterName=Loop_MainWheel})
  n29 = OneMinus(None=ComponentMask(None=@n15) {g=true})
  n30 = Add(
   A=Multiply(A=DotProduct(A=@n27, B=@n26), B=ScalarParameter() {defaultValue=1, group=TimeTrotter, parameterName=loops_Wing})
   B=Multiply(A=DotProduct(A=@n25, B=@n27), B=ScalarParameter() {defaultValue=1, group=TimeTrotter, parameterName=loops_SideWheel})
  )
  n31 = Add(
   A=Multiply(A=Time(), B=LinearInterpolate(A=@n13, B=@n14, Alpha=@n17))
   B=LinearInterpolate(A=Multiply(A=@n23, B=@n28), B=Multiply(A=@n29, B=@n30) {constB=3}, Alpha=@n17)
  )
  n32 = TextureSampleParameter2D(UVs=Frac(None=Add(A=@n12, B=@n31))) {
   group=TimeTrotter
   mipValueMode=TMVM_MipLevel
   parameterName=PanningColorGradient
   samplerSource=SSM_Wrap_WorldGroupSettings
   sortPriority=6
   texture=/Game/Global/Textures/Default/Masks_and_Gradients/3pxRGBImage
  }
  n33 = Multiply(
   A=Desaturation(None=@n32.RGB, Fraction=ScalarParameter() {group=TimeTrotter, parameterName=PanningColorGradient.Desaturation})
   B=NamedRerouteUsage()
  )
  n34 = ScalarParameter() {defaultValue=1, group=TimeTrotter, parameterName=PanningColorGradient.Brightness}
  n35 = Subtract(A=@n17, B=Subtract(A=Step(Y=@n16) {constX=0.352}, B=Step(Y=@n16) {constX=0.312}))
  n36 = Subtract(
   A=SphereMask(A=@n18, B=@n19) {attenuationRadius=0.25}
   B=SphereMask(A=@n18, B=@n19, Radius=ScalarParameter() {defaultValue=0.2, group=TimeTrotter, parameterName=CenterDiskMaskSize}) {attenuationRadius=0.2}
  )
  NamedRerouteDeclaration(None=Multiply(A=Multiply(A=@n33, B=@n34), B=Max(A=@n35, B=@n36))) {name=AdvancedEmissiveOutput, nodeColor=<rgba 1 0 0.2578 1>}
  n37 = Step(Y=ComponentMask(None=TextureCoordinate() {coordinateIndex=1}) {r=true}) {constX=0}
  NamedRerouteDeclaration(None=Add(A=OneMinus(None=@n37), B=@n7.G)) {name=UV2_Mask, nodeColor=<rgba 1 0.2578 0 1>}
  NamedRerouteDeclaration(None=Subtract(A=@n7.R, B=@n7.R)) {name=UV1_EmissiveAreaMask, nodeColor=<rgba 0 1 0.2969 1>}
  TemporalResponsivenessOutput!(None=@n4)
 ]
}
