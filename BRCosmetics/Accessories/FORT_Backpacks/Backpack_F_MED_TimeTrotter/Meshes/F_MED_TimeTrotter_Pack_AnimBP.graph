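# Animation Blueprint for the TimeTrotter back bling, posing and jiggling its wings as the player moves.
# A state machine selects named poses by movement state, with nested skydive states and acceleration-driven additive blendspaces in most states.
# The pose passes through an additive pack-bone transform into chained Anim Dynamics springs on both wings and their flaps.
# The event graph derives pack offsets from spine rotation, though the transform shows zero inputs.
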AnimBlueprint /BRCosmetics/Accessories/FORT_Backpacks/Backpack_F_MED_TimeTrotter/Meshes/F_MED_TimeTrotter_Pack_AnimBP
target_skeleton=&/F_MED_TimeTrotter_Pack_Skeleton
graphs [
 AnimGraph {nodes [
   AnimGraphNode_StateMachine_0 = Wing Position() {is=title AnimGraphNode_StateMachine}
   n1 = "Transform (Modify) Bone - Bone: pack"(ComponentPose=Local To Component(LocalPose=@AnimGraphNode_StateMachine_0), Translation="0,0,0", Rotation="0,0,0") {rotationMode=BMM_Additive, translationMode=BMM_Additive, alphaBoolBlend {blendOption=Linear}, boneToModify {boneName=pack}}
   n2 = Anim Dynamics - Bone: dyn_wing_01_r(ComponentPose=@n1, Alpha=1, ExternalForce="0,0,0") {
    angularDampingOverride=1
    angularSpringConstant=8000
    bAngularSpring=true
    bOverrideAngularDamping=true
    bUseGravityOverride=true
    gravityScale=0.35
    simulationSpace=RootRelative
    boundBone {boneName=dyn_wing_01_r}
    componentAppliedLinearAccClamp=<2000 2000 2000>
    componentLinearAccScale=<0.3 0.3 0.3>
    componentLinearVelScale=<0.3 0.3 0.3>
    gravityOverride {z=50}
    physicsBodyDefinitions [{
      boundBone {boneName=dyn_wing_01_r}
      boxExtents=<20 30 30>
      constraintSetup {angularLimitsMax=<0 5 5>, angularLimitsMin=<0 -5 -5>, angularTarget=<1 0 0>}
      localJointOffset=<2 0 0>
     }]
    simSpaceSettings {simSpaceAngularAlpha=0.3}
   }
   n3 = Anim Dynamics - Bone: dyn_wingFlap_04_r(ComponentPose=@n2, Alpha=1, ExternalForce="0,0,0") {
    angularDampingOverride=1
    angularSpringConstant=8000
    bAngularSpring=true
    bOverrideAngularDamping=true
    bUseGravityOverride=true
    gravityScale=0.35
    simulationSpace=RootRelative
    boundBone {boneName=dyn_wingFlap_04_r}
    componentAppliedLinearAccClamp=<2000 2000 2000>
    componentLinearAccScale=<0.3 0.3 0.3>
    componentLinearVelScale=<0.3 0.3 0.3>
    gravityOverride {z=50}
    physicsBodyDefinitions [{
      boundBone {boneName=dyn_wingFlap_04_r}
      boxExtents=<20 30 30>
      constraintSetup {angularLimitsMax=<0 5 8>, angularLimitsMin=<0 -5 -8>, angularTarget=<1 0 0>}
      localJointOffset=<4 0 0>
     }]
    simSpaceSettings {simSpaceAngularAlpha=0.3}
   }
   n4 = Anim Dynamics - Bone: dyn_wingFlap_03_r(ComponentPose=@n3, Alpha=1, ExternalForce="0,0,0") {
    angularDampingOverride=1
    angularSpringConstant=8000
    bAngularSpring=true
    bOverrideAngularDamping=true
    bUseGravityOverride=true
    gravityScale=0.35
    simulationSpace=RootRelative
    boundBone {boneName=dyn_wingFlap_03_r}
    componentAppliedLinearAccClamp=<2000 2000 2000>
    componentLinearAccScale=<0.3 0.3 0.3>
    componentLinearVelScale=<0.3 0.3 0.3>
    gravityOverride {z=50}
    physicsBodyDefinitions [{
      boundBone {boneName=dyn_wingFlap_03_r}
      boxExtents=<20 30 30>
      constraintSetup {angularLimitsMax=<0 5 8>, angularLimitsMin=<0 -5 -8>, angularTarget=<1 0 0>}
      localJointOffset=<4 0 0>
     }]
    simSpaceSettings {simSpaceAngularAlpha=0.3}
   }
   n5 = Anim Dynamics - Bone: dyn_wingFlap_02_r(ComponentPose=@n4, Alpha=1, ExternalForce="0,0,0") {
    angularDampingOverride=1
    angularSpringConstant=8000
    bAngularSpring=true
    bOverrideAngularDamping=true
    bUseGravityOverride=true
    gravityScale=0.35
    simulationSpace=RootRelative
    boundBone {boneName=dyn_wingFlap_02_r}
    componentAppliedLinearAccClamp=<2000 2000 2000>
    componentLinearAccScale=<0.3 0.3 0.3>
    componentLinearVelScale=<0.3 0.3 0.3>
    gravityOverride {z=50}
    physicsBodyDefinitions [{
      boundBone {boneName=dyn_wingFlap_02_r}
      boxExtents=<20 30 30>
      constraintSetup {angularLimitsMax=<0 5 8>, angularLimitsMin=<0 -5 -8>, angularTarget=<1 0 0>}
      localJointOffset=<4 0 0>
     }]
    simSpaceSettings {simSpaceAngularAlpha=0.3}
   }
   n6 = Anim Dynamics - Bone: dyn_wingFlap_01_r(ComponentPose=@n5, Alpha=1, ExternalForce="0,0,0") {
    angularDampingOverride=1
    angularSpringConstant=8000
    bAngularSpring=true
    bOverrideAngularDamping=true
    bUseGravityOverride=true
    gravityScale=0.35
    simulationSpace=RootRelative
    boundBone {boneName=dyn_wingFlap_01_r}
    componentAppliedLinearAccClamp=<2000 2000 2000>
    componentLinearAccScale=<0.3 0.3 0.3>
    componentLinearVelScale=<0.3 0.3 0.3>
    gravityOverride {z=50}
    physicsBodyDefinitions [{
      boundBone {boneName=dyn_wingFlap_01_r}
      boxExtents=<20 30 30>
      constraintSetup {angularLimitsMax=<0 5 8>, angularLimitsMin=<0 -5 -8>, angularTarget=<1 0 0>}
      localJointOffset=<4 0 0>
     }]
    simSpaceSettings {simSpaceAngularAlpha=0.3}
   }
   n7 = Anim Dynamics - Bone: dyn_wing_01_l(ComponentPose=@n6, Alpha=1, ExternalForce="0,0,0") {
    angularDampingOverride=1
    angularSpringConstant=8000
    bAngularSpring=true
    bOverrideAngularDamping=true
    bUseGravityOverride=true
    gravityScale=0.35
    simulationSpace=RootRelative
    boundBone {boneName=dyn_wing_01_l}
    componentAppliedLinearAccClamp=<2000 2000 2000>
    componentLinearAccScale=<0.3 0.3 0.3>
    componentLinearVelScale=<0.3 0.3 0.3>
    gravityOverride {z=50}
    physicsBodyDefinitions [{
      boundBone {boneName=dyn_wing_01_l}
      boxExtents=<20 30 30>
      constraintSetup {angularLimitsMax=<0 5 5>, angularLimitsMin=<0 -5 -5>, angularTarget=<1 0 0>}
      localJointOffset=<2 0 0>
     }]
    simSpaceSettings {simSpaceAngularAlpha=0.3}
   }
   n8 = Anim Dynamics - Bone: dyn_wingFlap_04_l(ComponentPose=@n7, Alpha=1, ExternalForce="0,0,0") {
    angularDampingOverride=1
    angularSpringConstant=8000
    bAngularSpring=true
    bOverrideAngularDamping=true
    bUseGravityOverride=true
    gravityScale=0.35
    simulationSpace=RootRelative
    boundBone {boneName=dyn_wingFlap_04_l}
    componentAppliedLinearAccClamp=<2000 2000 2000>
    componentLinearAccScale=<0.3 0.3 0.3>
    componentLinearVelScale=<0.3 0.3 0.3>
    gravityOverride {z=50}
    physicsBodyDefinitions [{
      boundBone {boneName=dyn_wingFlap_04_l}
      boxExtents=<20 30 30>
      constraintSetup {angularLimitsMax=<0 5 8>, angularLimitsMin=<0 -5 -8>, angularTarget=<1 0 0>}
      localJointOffset=<4 0 0>
     }]
    simSpaceSettings {simSpaceAngularAlpha=0.3}
   }
   n9 = Anim Dynamics - Bone: dyn_wingFlap_03_l(ComponentPose=@n8, Alpha=1, ExternalForce="0,0,0") {
    angularDampingOverride=1
    angularSpringConstant=8000
    bAngularSpring=true
    bOverrideAngularDamping=true
    bUseGravityOverride=true
    gravityScale=0.35
    simulationSpace=RootRelative
    boundBone {boneName=dyn_wingFlap_03_l}
    componentAppliedLinearAccClamp=<2000 2000 2000>
    componentLinearAccScale=<0.3 0.3 0.3>
    componentLinearVelScale=<0.3 0.3 0.3>
    gravityOverride {z=50}
    physicsBodyDefinitions [{
      boundBone {boneName=dyn_wingFlap_03_l}
      boxExtents=<20 30 30>
      constraintSetup {angularLimitsMax=<0 5 8>, angularLimitsMin=<0 -5 -8>, angularTarget=<1 0 0>}
      localJointOffset=<4 0 0>
     }]
    simSpaceSettings {simSpaceAngularAlpha=0.3}
   }
   n10 = Anim Dynamics - Bone: dyn_wingFlap_02_l(ComponentPose=@n9, Alpha=1, ExternalForce="0,0,0") {
    angularDampingOverride=1
    angularSpringConstant=8000
    bAngularSpring=true
    bOverrideAngularDamping=true
    bUseGravityOverride=true
    gravityScale=0.35
    simulationSpace=RootRelative
    boundBone {boneName=dyn_wingFlap_02_l}
    componentAppliedLinearAccClamp=<2000 2000 2000>
    componentLinearAccScale=<0.3 0.3 0.3>
    componentLinearVelScale=<0.3 0.3 0.3>
    gravityOverride {z=50}
    physicsBodyDefinitions [{
      boundBone {boneName=dyn_wingFlap_02_l}
      boxExtents=<20 30 30>
      constraintSetup {angularLimitsMax=<0 5 8>, angularLimitsMin=<0 -5 -8>, angularTarget=<1 0 0>}
      localJointOffset=<4 0 0>
     }]
    simSpaceSettings {simSpaceAngularAlpha=0.3}
   }
   n11 = Anim Dynamics - Bone: dyn_wingFlap_01_l(ComponentPose=@n10, Alpha=1, ExternalForce="0,0,0") {
    angularDampingOverride=1
    angularSpringConstant=8000
    bAngularSpring=true
    bOverrideAngularDamping=true
    bUseGravityOverride=true
    gravityScale=0.35
    simulationSpace=RootRelative
    boundBone {boneName=dyn_wingFlap_01_l}
    componentAppliedLinearAccClamp=<2000 2000 2000>
    componentLinearAccScale=<0.3 0.3 0.3>
    componentLinearVelScale=<0.3 0.3 0.3>
    gravityOverride {z=50}
    physicsBodyDefinitions [{
      boundBone {boneName=dyn_wingFlap_01_l}
      boxExtents=<20 30 30>
      constraintSetup {angularLimitsMax=<0 5 8>, angularLimitsMin=<0 -5 -8>, angularTarget=<1 0 0>}
      localJointOffset=<4 0 0>
     }]
    simSpaceSettings {simSpaceAngularAlpha=0.3}
   }
   Output Pose(Result=Component To Local(ComponentPose=@n11)) {layerGroup=None}
  ]}
 Wing Position {
  entry=Base
  path=AnimGraph.AnimGraphNode_StateMachine_0.Wing Position
  ids=id
  nodes [
   AnimStateNode_0 = Base() {is=title AnimStateNode}
   AnimStateNode_1 = Targeting() {is=title AnimStateNode}
   AnimStateNode_2 = Jumping() {is=title AnimStateNode}
   AnimStateNode_3 = Falling() {is=title AnimStateNode}
   AnimStateNode_4 = SkydiveMode() {is=title AnimStateNode}
   AnimStateNode_5 = Crouching() {is=title AnimStateNode}
   AnimStateNode_9 = Tactical Sprint() {is=title AnimStateNode}
   AnimStateTransitionNode_0 = AnimStateTransitionNode()
   AnimStateTransitionNode_1 = AnimStateTransitionNode()
   AnimStateTransitionNode_15 = AnimStateTransitionNode() {blendMode=CubicInOut, crossfadeDuration=1}
   AnimStateTransitionNode_16 = AnimStateTransitionNode() {customBlendCurve=/Game/Accessories/FORT_Backpacks/M_MED_Demon_Pack/Animation/JumpTransitionBlendOutCurve}
   AnimStateTransitionNode_17 = AnimStateTransitionNode() {customBlendCurve=/Game/Accessories/FORT_Backpacks/M_MED_Demon_Pack/Animation/JumpTransitionBlendCurve}
   AnimStateTransitionNode_18 = AnimStateTransitionNode()
   AnimStateTransitionNode_19 = AnimStateTransitionNode()
   AnimStateTransitionNode_2 = AnimStateTransitionNode() {
    blendMode=CubicInOut
    crossfadeDuration=1
    customBlendCurve=&JumpTransitionBlendCurve
   }
   AnimStateTransitionNode_20 = AnimStateTransitionNode()
   AnimStateTransitionNode_21 = AnimStateTransitionNode() {crossfadeDuration=0.25}
   AnimStateTransitionNode_22 = AnimStateTransitionNode() {crossfadeDuration=0.25}
   AnimStateTransitionNode_23 = AnimStateTransitionNode()
   AnimStateTransitionNode_24 = AnimStateTransitionNode()
   AnimStateTransitionNode_3 = AnimStateTransitionNode() {crossfadeDuration=0.15, customBlendCurve=&JumpTransitionBlendOutCurve}
   AnimStateTransitionNode_4 = AnimStateTransitionNode()
   AnimStateTransitionNode_5 = AnimStateTransitionNode() {crossfadeDuration=1}
   AnimStateTransitionNode_6 = AnimStateTransitionNode() {bAllowInertializationForSelfTransitions=true}
   AnimStateTransitionNode_7 = AnimStateTransitionNode() {bAllowInertializationForSelfTransitions=true}
  ]
  links [
   {
    dst=Crouching
    rule=AnimGraph.AnimGraphNode_StateMachine_0.Wing Position.AnimStateTransitionNode_21.AnimationTransitionGraph_1
    src=Base
    via=AnimStateTransitionNode_21
   }
   {
    dst=Falling
    rule=AnimGraph.AnimGraphNode_StateMachine_0.Wing Position.AnimStateTransitionNode_17.AnimationTransitionGraph_1
    src=Base
    via=AnimStateTransitionNode_17
   }
   {
    dst=Jumping
    rule=AnimGraph.AnimGraphNode_StateMachine_0.Wing Position.AnimStateTransitionNode_2.AnimationTransitionGraph_1
    src=Base
    via=AnimStateTransitionNode_2
   }
   {
    dst=SkydiveMode
    rule=AnimGraph.AnimGraphNode_StateMachine_0.Wing Position.AnimStateTransitionNode_20.AnimationTransitionGraph_1
    src=Base
    via=AnimStateTransitionNode_20
   }
   {
    dst=Tactical Sprint
    rule=AnimGraph.AnimGraphNode_StateMachine_0.Wing Position.AnimStateTransitionNode_0.Transition
    src=Base
    via=AnimStateTransitionNode_0
   }
   {
    dst=Targeting
    rule=AnimGraph.AnimGraphNode_StateMachine_0.Wing Position.AnimStateTransitionNode_23.AnimationTransitionGraph_1
    src=Base
    via=AnimStateTransitionNode_23
   }
   {
    dst=Base
    rule=AnimGraph.AnimGraphNode_StateMachine_0.Wing Position.AnimStateTransitionNode_22.AnimationTransitionGraph_1
    src=Crouching
    via=AnimStateTransitionNode_22
   }
   {
    dst=Targeting
    rule=AnimGraph.AnimGraphNode_StateMachine_0.Wing Position.AnimStateTransitionNode_6.Transition
    src=Crouching
    via=AnimStateTransitionNode_6
   }
   {
    dst=Base
    rule=AnimGraph.AnimGraphNode_StateMachine_0.Wing Position.AnimStateTransitionNode_16.AnimationTransitionGraph_1
    src=Falling
    via=AnimStateTransitionNode_16
   }
   {
    dst=Base
    rule=AnimGraph.AnimGraphNode_StateMachine_0.Wing Position.AnimStateTransitionNode_3.AnimationTransitionGraph_1
    src=Jumping
    via=AnimStateTransitionNode_3
   }
   {
    dst=Falling
    rule=AnimGraph.AnimGraphNode_StateMachine_0.Wing Position.AnimStateTransitionNode_15.AnimationTransitionGraph_1
    src=Jumping
    via=AnimStateTransitionNode_15
   }
   {
    dst=SkydiveMode
    rule=AnimGraph.AnimGraphNode_StateMachine_0.Wing Position.AnimStateTransitionNode_18.AnimationTransitionGraph_1
    src=Jumping
    via=AnimStateTransitionNode_18
   }
   {
    dst=Tactical Sprint
    rule=AnimGraph.AnimGraphNode_StateMachine_0.Wing Position.AnimStateTransitionNode_4.Transition
    src=Jumping
    via=AnimStateTransitionNode_4
   }
   {
    dst=Base
    rule=AnimGraph.AnimGraphNode_StateMachine_0.Wing Position.AnimStateTransitionNode_19.AnimationTransitionGraph_1
    src=SkydiveMode
    via=AnimStateTransitionNode_19
   }
   {
    dst=Base
    rule=AnimGraph.AnimGraphNode_StateMachine_0.Wing Position.AnimStateTransitionNode_1.Transition
    src=Tactical Sprint
    via=AnimStateTransitionNode_1
   }
   {
    dst=Jumping
    rule=AnimGraph.AnimGraphNode_StateMachine_0.Wing Position.AnimStateTransitionNode_5.Transition
    src=Tactical Sprint
    via=AnimStateTransitionNode_5
   }
   {
    dst=Base
    rule=AnimGraph.AnimGraphNode_StateMachine_0.Wing Position.AnimStateTransitionNode_24.AnimationTransitionGraph_1
    src=Targeting
    via=AnimStateTransitionNode_24
   }
   {
    dst=Crouching
    rule=AnimGraph.AnimGraphNode_StateMachine_0.Wing Position.AnimStateTransitionNode_7.Transition
    src=Targeting
    via=AnimStateTransitionNode_7
   }
  ]
 }
 Base {
  path=AnimGraph.AnimGraphNode_StateMachine_0.Wing Position.AnimStateNode_0.Base
  nodes [
   n1 = Get PlayerSidewaysAcceleration() -> (PlayerSidewaysAcceleration:float)
   n2 = Get PlayerForwardAcceleration() -> (PlayerForwardAcceleration:float)
   n3 = TimeTrotter_Wing_Movement_Poses_PoseAsset : Base!(PoseWeight:float=1) -> (Pose:Pose Link Structure) {
    is=title AnimGraphNode_PoseByName
    poseAsset=/BRCosmetics/Accessories/FORT_Backpacks/Backpack_F_MED_TimeTrotter/Animation/TimeTrotter_Wing_Movement_Poses_PoseAsset
    poseName=Base
   }
   Output Animation Pose(Result=Apply Additive(
     Base=@n3
     Additive=Blendspace Player 'BS_TimeTrotter_Wing_Ground'(X=@n1, Y=@n2) {blendSpace=/BRCosmetics/Accessories/FORT_Backpacks/Backpack_F_MED_TimeTrotter/Animation/BS_TimeTrotter_Wing_Ground}
    ) {alphaBoolBlend {blendOption=Linear}})
  ]
 }
 Targeting {
  path=AnimGraph.AnimGraphNode_StateMachine_0.Wing Position.AnimStateNode_1.Targeting
  nodes [
   n1 = TimeTrotter_Wing_Movement_Poses_PoseAsset : Targeting!(PoseWeight:float=1) -> (Pose:Pose Link Structure) {
    is=title AnimGraphNode_PoseByName
    poseAsset=&TimeTrotter_Wing_Movement_Poses_PoseAsset
    poseName=Targeting
   }
   Output Animation Pose(Result=@n1)
  ]
 }
 Jumping {
  path=AnimGraph.AnimGraphNode_StateMachine_0.Wing Position.AnimStateNode_2.Jumping
  nodes [
   n1 = Get PlayerSidewaysAcceleration()
   n2 = Get PlayerForwardAcceleration()
   n3 = TimeTrotter_Wing_Movement_Poses_PoseAsset : Jump!(PoseWeight:float=1) -> (Pose:Pose Link Structure) {
    is=title AnimGraphNode_PoseByName
    poseAsset=&TimeTrotter_Wing_Movement_Poses_PoseAsset
    poseName=Jump
   }
   Output Animation Pose(Result=Apply Additive(
     Base=@n3
     Additive=Blendspace Player 'BS_TimeTrotter_Wing_Ground'(X=@n1, Y=@n2) {blendSpace=&BS_TimeTrotter_Wing_Ground}
     Alpha=0.6
    ) {alphaBoolBlend {blendOption=Linear}})
  ]
 }
 Falling {
  path=AnimGraph.AnimGraphNode_StateMachine_0.Wing Position.AnimStateNode_3.Falling
  nodes [
   n1 = Get PlayerSidewaysAcceleration()
   n2 = Get PlayerForwardAcceleration()
   n3 = TimeTrotter_Wing_Movement_Poses_PoseAsset : Jump(PoseWeight=1) {
    poseAsset=&TimeTrotter_Wing_Movement_Poses_PoseAsset
    poseName=Jump
   }
   Output Animation Pose(Result=Apply Additive(
     Base=@n3
     Additive=Blendspace Player 'BS_TimeTrotter_Wing_Ground'(X=@n1, Y=@n2) {blendSpace=&BS_TimeTrotter_Wing_Ground}
     Alpha=0.6
    ) {alphaBoolBlend {blendOption=Linear}})
  ]
 }
 SkydiveMode {
  path=AnimGraph.AnimGraphNode_StateMachine_0.Wing Position.AnimStateNode_4.SkydiveMode
  nodes [
   AnimGraphNode_StateMachine_0 = SkyDivingState() {is=title AnimGraphNode_StateMachine}
   Output Animation Pose(Result=@AnimGraphNode_StateMachine_0)
  ]
 }
 SkyDivingState {
  entry=Parashuting
  path=AnimGraph.AnimGraphNode_StateMachine_0.Wing Position.AnimStateNode_4.SkydiveMode.AnimGraphNode_StateMachine_0.SkyDivingState
  ids=id
  nodes [
   AnimStateNode_0 = Parashuting() {is=title AnimStateNode}
   AnimStateNode_1 = Diving() {is=title AnimStateNode}
   AnimStateNode_2 = floating() {is=title AnimStateNode}
   AnimStateTransitionNode_0 = AnimStateTransitionNode()
   AnimStateTransitionNode_1 = AnimStateTransitionNode()
   AnimStateTransitionNode_3 = AnimStateTransitionNode()
   AnimStateTransitionNode_4 = AnimStateTransitionNode()
   AnimStateTransitionNode_5 = AnimStateTransitionNode()
   AnimStateTransitionNode_6 = AnimStateTransitionNode()
  ]
  links [
   {
    dst=Parashuting
    rule=AnimGraph.AnimGraphNode_StateMachine_0.Wing Position.AnimStateNode_4.SkydiveMode.AnimGraphNode_StateMachine_0.SkyDivingState.AnimStateTransitionNode_1.Transition
    src=Diving
    via=AnimStateTransitionNode_1
   }
   {
    dst=floating
    rule=AnimGraph.AnimGraphNode_StateMachine_0.Wing Position.AnimStateNode_4.SkydiveMode.AnimGraphNode_StateMachine_0.SkyDivingState.AnimStateTransitionNode_3.Transition
    src=Diving
    via=AnimStateTransitionNode_3
   }
   {
    dst=Diving
    rule=AnimGraph.AnimGraphNode_StateMachine_0.Wing Position.AnimStateNode_4.SkydiveMode.AnimGraphNode_StateMachine_0.SkyDivingState.AnimStateTransitionNode_0.Transition
    src=Parashuting
    via=AnimStateTransitionNode_0
   }
   {
    dst=floating
    rule=AnimGraph.AnimGraphNode_StateMachine_0.Wing Position.AnimStateNode_4.SkydiveMode.AnimGraphNode_StateMachine_0.SkyDivingState.AnimStateTransitionNode_6.Transition
    src=Parashuting
    via=AnimStateTransitionNode_6
   }
   {
    dst=Diving
    rule=AnimGraph.AnimGraphNode_StateMachine_0.Wing Position.AnimStateNode_4.SkydiveMode.AnimGraphNode_StateMachine_0.SkyDivingState.AnimStateTransitionNode_4.Transition
    src=floating
    via=AnimStateTransitionNode_4
   }
   {
    dst=Parashuting
    rule=AnimGraph.AnimGraphNode_StateMachine_0.Wing Position.AnimStateNode_4.SkydiveMode.AnimGraphNode_StateMachine_0.SkyDivingState.AnimStateTransitionNode_5.Transition
    src=floating
    via=AnimStateTransitionNode_5
   }
  ]
 }
 Parashuting {
  path=AnimGraph.AnimGraphNode_StateMachine_0.Wing Position.AnimStateNode_4.SkydiveMode.AnimGraphNode_StateMachine_0.SkyDivingState.AnimStateNode_0.Parashuting
  nodes [
   n1 = Get PlayerSidewaysAcceleration()
   n2 = Get PlayerForwardAcceleration()
   n3 = TimeTrotter_Wing_Movement_Poses_PoseAsset : Float!(PoseWeight:float=1) -> (Pose:Pose Link Structure) {
    is=title AnimGraphNode_PoseByName
    poseAsset=&TimeTrotter_Wing_Movement_Poses_PoseAsset
    poseName=Float
   }
   Output Animation Pose(Result=Apply Additive(
     Base=@n3
     Additive=Blendspace Player 'BS_TimeTrotter_Wing_Air'(X=@n1, Y=@n2) {blendSpace=/BRCosmetics/Accessories/FORT_Backpacks/Backpack_F_MED_TimeTrotter/Animation/BS_TimeTrotter_Wing_Air}
    ) {alphaBoolBlend {blendOption=Linear}})
  ]
 }
 Diving {
  path=AnimGraph.AnimGraphNode_StateMachine_0.Wing Position.AnimStateNode_4.SkydiveMode.AnimGraphNode_StateMachine_0.SkyDivingState.AnimStateNode_1.Diving
  nodes [
   n1 = Get PlayerSidewaysAcceleration()
   n2 = Get PlayerForwardAcceleration()
   n3 = TimeTrotter_Wing_Movement_Poses_PoseAsset : Dive!(PoseWeight:float=1) -> (Pose:Pose Link Structure) {
    is=title AnimGraphNode_PoseByName
    poseAsset=&TimeTrotter_Wing_Movement_Poses_PoseAsset
    poseName=Dive
   }
   Output Animation Pose(Result=Apply Additive(
     Base=@n3
     Additive=Blendspace Player 'BS_TimeTrotter_Wing_Air'(X=@n1, Y=@n2) {blendSpace=&BS_TimeTrotter_Wing_Air}
     Alpha=0.5
    ) {alphaBoolBlend {blendOption=Linear}})
  ]
 }
 floating {
  path=AnimGraph.AnimGraphNode_StateMachine_0.Wing Position.AnimStateNode_4.SkydiveMode.AnimGraphNode_StateMachine_0.SkyDivingState.AnimStateNode_2.floating
  nodes [
   n1 = Get PlayerSidewaysAcceleration()
   n2 = Get PlayerForwardAcceleration()
   n3 = TimeTrotter_Wing_Movement_Poses_PoseAsset : Float(PoseWeight=1) {
    poseAsset=&TimeTrotter_Wing_Movement_Poses_PoseAsset
    poseName=Float
   }
   Output Animation Pose(Result=Apply Additive(
     Base=@n3
     Additive=Blendspace Player 'BS_TimeTrotter_Wing_Air'(X=@n1, Y=@n2) {blendSpace=&BS_TimeTrotter_Wing_Air}
    ) {alphaBoolBlend {blendOption=Linear}})
  ]
 }
 Transition {
  path=AnimGraph.AnimGraphNode_StateMachine_0.Wing Position.AnimStateNode_4.SkydiveMode.AnimGraphNode_StateMachine_0.SkyDivingState.AnimStateTransitionNode_0.Transition
  nodes [Result(bCanEnterTransition=NOT Boolean(A=Get bIsParachuteOpened()))]
 }
 Transition {
  path=AnimGraph.AnimGraphNode_StateMachine_0.Wing Position.AnimStateNode_4.SkydiveMode.AnimGraphNode_StateMachine_0.SkyDivingState.AnimStateTransitionNode_1.Transition
  nodes [Result(bCanEnterTransition=Get bIsParachuteOpened())]
 }
 Transition {
  path=AnimGraph.AnimGraphNode_StateMachine_0.Wing Position.AnimStateNode_4.SkydiveMode.AnimGraphNode_StateMachine_0.SkyDivingState.AnimStateTransitionNode_3.Transition
  nodes [Result(bCanEnterTransition=NOT Boolean(A=Get bIsSkydiveDiveMode()))]
 }
 Transition {
  path=AnimGraph.AnimGraphNode_StateMachine_0.Wing Position.AnimStateNode_4.SkydiveMode.AnimGraphNode_StateMachine_0.SkyDivingState.AnimStateTransitionNode_4.Transition
  nodes [Result(bCanEnterTransition=Get bIsSkydiveDiveMode())]
 }
 Transition {
  path=AnimGraph.AnimGraphNode_StateMachine_0.Wing Position.AnimStateNode_4.SkydiveMode.AnimGraphNode_StateMachine_0.SkyDivingState.AnimStateTransitionNode_5.Transition
  nodes [Result(bCanEnterTransition=Get bIsParachuteOpened())]
 }
 Transition {
  path=AnimGraph.AnimGraphNode_StateMachine_0.Wing Position.AnimStateNode_4.SkydiveMode.AnimGraphNode_StateMachine_0.SkyDivingState.AnimStateTransitionNode_6.Transition
  nodes [Result(bCanEnterTransition=NOT Boolean(A=Get bIsParachuteOpened()))]
 }
 Crouching {
  path=AnimGraph.AnimGraphNode_StateMachine_0.Wing Position.AnimStateNode_5.Crouching
  nodes [
   n1 = Get PlayerSidewaysAcceleration()
   n2 = Get PlayerForwardAcceleration()
   n3 = TimeTrotter_Wing_Movement_Poses_PoseAsset : Crouch!(PoseWeight:float=1) -> (Pose:Pose Link Structure) {
    is=title AnimGraphNode_PoseByName
    poseAsset=&TimeTrotter_Wing_Movement_Poses_PoseAsset
    poseName=Crouch
   }
   Output Animation Pose(Result=Apply Additive(
     Base=@n3
     Additive=Blendspace Player 'BS_TimeTrotter_Wing_Ground'(X=@n1, Y=@n2) {blendSpace=&BS_TimeTrotter_Wing_Ground}
    ) {alphaBoolBlend {blendOption=Linear}})
  ]
 }
 Tactical Sprint {
  path=AnimGraph.AnimGraphNode_StateMachine_0.Wing Position.AnimStateNode_9.Tactical Sprint
  nodes [
   n1 = TimeTrotter_Wing_Movement_Poses_PoseAsset : Sprint!(PoseWeight:float=1) -> (Pose:Pose Link Structure) {
    is=title AnimGraphNode_PoseByName
    poseAsset=&TimeTrotter_Wing_Movement_Poses_PoseAsset
    poseName=Sprint
   }
   Output Animation Pose(Result=@n1)
  ]
 }
 Transition {
  path=AnimGraph.AnimGraphNode_StateMachine_0.Wing Position.AnimStateTransitionNode_0.Transition
  nodes [Result(bCanEnterTransition=Get bIsTacticalSprinting())]
 }
 Transition {
  path=AnimGraph.AnimGraphNode_StateMachine_0.Wing Position.AnimStateTransitionNode_1.Transition
  nodes [Result(bCanEnterTransition=NOT Boolean(A=Get bIsTacticalSprinting()))]
 }
 AnimationTransitionGraph_1 {
  path=AnimGraph.AnimGraphNode_StateMachine_0.Wing Position.AnimStateTransitionNode_15.AnimationTransitionGraph_1
  nodes [Result(bCanEnterTransition=Get bIsFalling())]
 }
 AnimationTransitionGraph_1 {
  path=AnimGraph.AnimGraphNode_StateMachine_0.Wing Position.AnimStateTransitionNode_16.AnimationTransitionGraph_1
  nodes [Result(bCanEnterTransition=NOT Boolean(A=Get bIsFalling()))]
 }
 AnimationTransitionGraph_1 {
  path=AnimGraph.AnimGraphNode_StateMachine_0.Wing Position.AnimStateTransitionNode_17.AnimationTransitionGraph_1
  nodes [Result(bCanEnterTransition=Get bIsFalling())]
 }
 AnimationTransitionGraph_1 {
  path=AnimGraph.AnimGraphNode_StateMachine_0.Wing Position.AnimStateTransitionNode_18.AnimationTransitionGraph_1
  nodes [Result(bCanEnterTransition=Get bIsSkyDiving())]
 }
 AnimationTransitionGraph_1 {
  path=AnimGraph.AnimGraphNode_StateMachine_0.Wing Position.AnimStateTransitionNode_19.AnimationTransitionGraph_1
  nodes [Result(bCanEnterTransition=NOT Boolean(A=Get bIsSkyDiving()))]
 }
 AnimationTransitionGraph_1 {
  path=AnimGraph.AnimGraphNode_StateMachine_0.Wing Position.AnimStateTransitionNode_2.AnimationTransitionGraph_1
  nodes [Result(bCanEnterTransition=Get bIsJumping())]
 }
 AnimationTransitionGraph_1 {
  path=AnimGraph.AnimGraphNode_StateMachine_0.Wing Position.AnimStateTransitionNode_20.AnimationTransitionGraph_1
  nodes [Result(bCanEnterTransition=Get bIsSkyDiving())]
 }
 AnimationTransitionGraph_1 {
  path=AnimGraph.AnimGraphNode_StateMachine_0.Wing Position.AnimStateTransitionNode_21.AnimationTransitionGraph_1
  nodes [Result(bCanEnterTransition=Get bIsCrouching())]
 }
 AnimationTransitionGraph_1 {
  path=AnimGraph.AnimGraphNode_StateMachine_0.Wing Position.AnimStateTransitionNode_22.AnimationTransitionGraph_1
  nodes [Result(bCanEnterTransition=NOT Boolean(A=Get bIsCrouching()))]
 }
 AnimationTransitionGraph_1 {
  path=AnimGraph.AnimGraphNode_StateMachine_0.Wing Position.AnimStateTransitionNode_23.AnimationTransitionGraph_1
  nodes [Result(bCanEnterTransition=Get bIsTargeting())]
 }
 AnimationTransitionGraph_1 {
  path=AnimGraph.AnimGraphNode_StateMachine_0.Wing Position.AnimStateTransitionNode_24.AnimationTransitionGraph_1
  nodes [Result(bCanEnterTransition=NOT Boolean(A=Get bIsTargeting()))]
 }
 AnimationTransitionGraph_1 {
  path=AnimGraph.AnimGraphNode_StateMachine_0.Wing Position.AnimStateTransitionNode_3.AnimationTransitionGraph_1
  nodes [Result(bCanEnterTransition=NOT Boolean(A=Get bIsJumping()))]
 }
 Transition {
  path=AnimGraph.AnimGraphNode_StateMachine_0.Wing Position.AnimStateTransitionNode_4.Transition
  nodes [Result(bCanEnterTransition=NOT Boolean(A=Get bIsJumping()))]
 }
 Transition {
  path=AnimGraph.AnimGraphNode_StateMachine_0.Wing Position.AnimStateTransitionNode_5.Transition
  nodes [Result(bCanEnterTransition=Get bIsJumping())]
 }
 Transition {
  path=AnimGraph.AnimGraphNode_StateMachine_0.Wing Position.AnimStateTransitionNode_6.Transition
  nodes [Result(bCanEnterTransition=Get bIsTargeting())]
 }
 Transition {
  path=AnimGraph.AnimGraphNode_StateMachine_0.Wing Position.AnimStateTransitionNode_7.Transition
  nodes [Result(bCanEnterTransition=NOT Boolean(A=Get bIsTargeting()))]
 }
 EventGraph {nodes [
   Event BlueprintUpdateAnimation()
   > Sequence!(execute) -> (then_0, then_1, then_2:Exec)
   n1 = GetOwnerSkeletalMeshComponent()
   n2 = BreakRotator(InRot=BreakTransform(InTransform=GetDeltaTransformFromRefPose(self=@n1, BoneName=spine_05, BaseName=spine_01)).Rotation)
   > Set backpack_rotator(backpack_rotator:Rotator=MakeRotator(Roll=MapRangeClamped(Value=@n2.Yaw, InRangeA=0, InRangeB=-30, OutRangeA=5, OutRangeB=10))) -> (Output_Get:Rotator)
   > Set backpack_translation(backpack_translation=MakeVector(Y=MapRangeClamped(Value=@n2.Yaw, InRangeA=0, InRangeB=-20, OutRangeA=-0.5, OutRangeB=-2)))
   > Set Skydiving(Skydiving:double="To Float (Boolean)"(InBool=Get bIsSkyDiving())) -> (Output_Get:double)
   TryGetPawnOwner()
  ]}
]
variables [
 backpack_rotator {type=Rotator}
 backpack_translation {type=Vector}
 EmotePlaying {type=number}
 Skydiving {type=number}
]
